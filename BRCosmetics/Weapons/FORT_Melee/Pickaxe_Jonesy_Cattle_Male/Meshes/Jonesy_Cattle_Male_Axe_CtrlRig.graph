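# Control Rig for the Jonesy Cattle Male pickaxe, whose skeleton has a cattle head with mouth, nose, ear and horn bones on a two-handed handle.
# Its single forward-solve step reads neck_c's current local X translation and writes it back with Y and Z zeroed, identity rotation and unit scale.
# This restricts the neck to sliding along one axis relative to neck_c_base, with the head bones parented beneath it.
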ControlRigBlueprint /BRCosmetics/Weapons/FORT_Melee/Pickaxe_Jonesy_Cattle_Male/Meshes/Jonesy_Cattle_Male_Axe_CtrlRig
graphs [
 RigVMFunctionLibrary {links [], nodes []}
 RigVMModel {nodes [
   RigUnit_BeginExecution = &RigUnit_BeginExecution()
   &RigUnit_SetTransform(
    Item=(Type=Bone,Name="neck_c")
    Space=LocalSpace
    Value=(Rotation=(X=0,Y=0,Z=0,W=1),Translation=(X=0,Y=0,Z=0),Scale3D=(X=1,Y=1,Z=1))
    Weight=1
    bInitial=false
    Value.Translation.X=&RigUnit_GetTransform(Item=(Type=Bone,Name="neck_c"), Space=LocalSpace, bInitial=false).Transform.Translation.X
    ExecutePin=@RigUnit_BeginExecution.ExecutePin
   )
  ]}
]
hierarchy [
 b_AR_Root {loc=<0 0 0>, rot=<0 0 0.7071 0.7071>}
 Melee_TwoHanded_Handle {parent=b_AR_Root, loc=<8.1477 -0.008 -29.0793>, rot=<0 0 0.7071 -0.7071>}
 neck_c_base {parent=Melee_TwoHanded_Handle, loc=<0.7109 -1.1717 67.9564>, rot=<0 0.7071 0 -0.7071>}
 neck_c {parent=neck_c_base, loc=<0 0 0>, rot=<0 0 0 -1>}
 head_c {parent=neck_c, loc=<9.2434 0 0>, rot=<0 0 0 -1>}
 mouth {parent=head_c, loc=<11.2547 15.4027 0.7112>, rot=<0.4587 0.5394 0.5471 0.4465>}
 nose_l {parent=mouth, loc=<2.0744 -10.5906 10.0441>, rot=<0.2163 0.2639 0.731 -0.5909>}
 nose_r {parent=mouth, loc=<2.5021 10.5658 9.9889>, rot=<0.7196 -0.6062 -0.2094 -0.2663>}
 ear_l {parent=head_c, loc=<17.8533 -1.3714 -18.1655>, rot=<0.0602 -0.7045 -0.0602 -0.7045>}
 ear_r {parent=head_c, loc=<17.8533 -1.3714 19.6033>, rot=<0.7045 0.0602 -0.7045 0.0602>}
 horn_l {parent=head_c, loc=<23.9023 -8.9369 -13.9472>, rot=<0.0953 -0.5177 0.1718 -0.8327>}
 horn_r {parent=head_c, loc=<23.9023 -8.9369 15.385>, rot=<-0.5177 -0.0953 0.8327 0.1718>}
]
