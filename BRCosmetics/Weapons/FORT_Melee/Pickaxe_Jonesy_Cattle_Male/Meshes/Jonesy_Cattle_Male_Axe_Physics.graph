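# Physics asset for the Jonesy Cattle male pickaxe, giving its two-handed handle and animal head a jointed capsule rig.
# The neck hangs off the handle with free swing and twist, while the head is tightly limited to the neck.
# Ears, horns and mouth constrain to the head, and both nose bodies to the mouth.
# Narrow angular limits allow small secondary movement, loosest at the ears, with mouth twist locked.
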PhysicsAsset /BRCosmetics/Weapons/FORT_Melee/Pickaxe_Jonesy_Cattle_Male/Meshes/Jonesy_Cattle_Male_Axe_Physics
bodies {
 Melee_TwoHanded_Handle [{shapeType=Capsule, length=50.3192, radius=4.0847, center=<0.7311 -1.1793 9.8899>, rotation=<rot 0.0107 -90 -0.1083>}]
 ear_l [{shapeType=Capsule, length=8.75, radius=4.75, center=<6 0 0>, rotation=<rot 90 0 0>}]
 ear_r [{shapeType=Capsule, length=8.75, radius=4.75, center=<-6 0 0>, rotation=<rot 90 0 180>}]
 head_c [{shapeType=Capsule, length=9.5, radius=10.25, center=<16.085 0 -0.0003>, rotation=<rot 90 0 0>}]
 horn_l [{shapeType=Capsule, length=6.5, radius=4.25, center=<6.6741 0.0001 0>, rotation=<rot 90 0 0>}]
 horn_r [{shapeType=Capsule, length=6.5, radius=4.25, center=<-6.6741 -0.0001 0>, rotation=<rot 90 0 180>}]
 mouth [{shapeType=Capsule, length=4.75, radius=9, center=<0 0 -2>}]
 neck_c [{shapeType=Capsule, length=1, radius=1}]
 nose_l [{shapeType=Capsule, length=3.5, radius=2.5, center=<4 0.0001 0>, rotation=<rot 90 0 0>}]
 nose_r [{shapeType=Capsule, length=3.5, radius=2.5, center=<-4 0 0>, rotation=<rot 90 0 180>}]
}
constraints [
 {bone1Name=nose_l, bone2Name=mouth, swing1LimitDegrees=30, swing2LimitDegrees=5, twistLimitDegrees=5}
 {bone1Name=nose_r, bone2Name=mouth, swing1LimitDegrees=30, swing2LimitDegrees=5, twistLimitDegrees=5}
 {bone1Name=mouth, bone2Name=head_c, swing1LimitDegrees=5, swing2LimitDegrees=5, twistLimitDegrees=5, twistMotion=Locked}
 {bone1Name=ear_l, bone2Name=head_c, swing1LimitDegrees=45, swing2LimitDegrees=40, twistLimitDegrees=5}
 {bone1Name=ear_r, bone2Name=head_c, swing1LimitDegrees=45, swing2LimitDegrees=40, twistLimitDegrees=5}
 {bone1Name=head_c, bone2Name=neck_c, swing1LimitDegrees=2, swing2LimitDegrees=2, twistLimitDegrees=2}
 {bone1Name=horn_l, bone2Name=head_c, swing1LimitDegrees=10, swing2LimitDegrees=20, twistLimitDegrees=5}
 {bone1Name=horn_r, bone2Name=head_c, swing1LimitDegrees=10, swing2LimitDegrees=20, twistLimitDegrees=5}
 {
  bone1Name=neck_c
  bone2Name=Melee_TwoHanded_Handle
  swing1LimitDegrees=45
  swing1Motion=Free
  swing2LimitDegrees=45
  swing2Motion=Free
  twistLimitDegrees=45
  twistMotion=Free
 }
]
