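# Animation Blueprint for the Jonesy Cattle Male pickaxe, adding procedural secondary motion to the weapon mesh.
# The input pose passes through a neck_c spring, a Control Rig, then RigidBody physics simulated relative to b_AR_Root.
# Each update raises spring alpha during melee animations and multiplies it to zero while riding.
# Jumping or falling feeds an upward external force into the physics node.
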AnimBlueprint /BRCosmetics/Weapons/FORT_Melee/Pickaxe_Jonesy_Cattle_Male/Meshes/Jonesy_Cattle_Male_Axe_AnimBP
target_skeleton=&/Jonesy_Cattle_Male_Axe_Skeleton
graphs [
 AnimGraph {nodes [
   n1 = Spring controller - Bone: neck_c!(
    ComponentPose:Component Space Pose Link Structure=Local To Component(LocalPose=Input Pose())
    Alpha:float=Get Spring Alpha Not Riding() -> (Spring Alpha Not Riding:double)
    SpringDamping:double=Get springDampening() -> (springDampening:double)
   ) -> (Pose:Component Space Pose Link Structure) {
    is=title AnimGraphNode_SpringBone
    bLimitDisplacement=true
    maxDisplacement=18
    springDamping=50
    springStiffness=500
    alphaBoolBlend {blendOption=Linear}
    alphaScaleBiasClamp {bInterpResult=true, interpSpeedDecreasing=20, interpSpeedIncreasing=40}
    springBone {boneName=neck_c}
   }
   n2 = Control Rig(Source=Component To Local(ComponentPose=@n1)) {
    alphaBoolBlend {blendOption=Linear}
    controlRigAssetReference {blueprintRigClass=&/Jonesy_Cattle_Male_Axe_CtrlRig.Jonesy_Cattle_Male_Axe_CtrlRig_C}
   }
   n3 = RigidBody(ComponentPose=Local To Component(LocalPose=@n2), ExternalForce=Get Jump_Impulse() -> (Jump_Impulse:Vector)) {
    bClampLinearTranslationLimitToRefPose=true
    bOverrideWorldGravity=true
    overridePhysicsAsset=&/Jonesy_Cattle_Male_Axe_Physics
    simulationSpace=BaseBoneSpace
    alphaBoolBlend {blendOption=Linear}
    baseBoneRef {boneName=b_AR_Root}
    componentAppliedLinearAccClamp=<6000 6000 6000>
    componentLinearAccScale=<0.3 0.3 0.3>
    componentLinearVelScale=<0.3 0.3 0.3>
    overrideWorldGravity {z=-980}
   }
   Output Pose(Result=Component To Local(ComponentPose=@n3)) {layerGroup=None}
   Get springAlpha() -> (springAlpha:double)
  ]}
 EventGraph {nodes [
   Event BlueprintUpdateAnimation()
   > n1 = Sequence!(execute) -> (then_0, then_1, then_2:Exec, then_3:Exec)
   > n2 = Branch(Condition=Get bIsPlayingMeleeAnim())
   > Set springAlpha(springAlpha:double=0.3) -> (Output_Get:double)
   > Set springDampening(springDampening:double=5) -> (Output_Get:double)
   Set springAlpha(execute=@n2.else, springAlpha=0.2)
   > Set springDampening(springDampening=5)
   n3 = Get bIsRiding() -> (bIsRiding:bool)
   Set Is Not Riding Float(execute=@n1.then_2, Is Not Riding Float:float="To Float (Boolean)"(InBool=NOT Boolean(A=@n3))) -> (Output_Get:float)
   Set Spring Alpha Not Riding(
    execute=@n1.then_3
    Spring Alpha Not Riding:double=float * float(A=Get springAlpha(), B=Get Is Not Riding Float() -> (Is Not Riding Float:float)) {is=title K2Node_PromotableOperator}
   ) -> (Output_Get:double)
   n4 = Get bIsJumping()
   n5 = Get bIsFalling()
   Set Jump_Impulse(execute=@n1.then_1, Jump_Impulse:Vector=SelectVector(A="0, 0,200", B="0, 0, 0", bPickA=OR Boolean(A=@n4, B=@n5))) -> (Output_Get:Vector)
   TryGetPawnOwner()
  ]}
]
variables [
 Is Not Riding Float {type=number}
 Jump_Impulse {type=Vector}
 Spring Alpha Not Riding {type=number}
 springAlpha {type=number}
 springDampening {type=number}
]
